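; Blueprint for the RATT pistol weapon, adding audio and visual feedback that depends on aim-down-sights state.
; On targeting changes it stores PlayerIsADS and plays the ADS or hip-fire transition sound attached to the weapon mesh.
; BeginPlay binds OnStartFiring to a handler that attaches a reload-smoke Niagara effect at the Muzzle socket, only while aiming.
; The cosmetic-events check gates nothing, and both muzzle-flash OnRep functions are empty.
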(event EventBeginPlay
  (|CallParentFunction)
  (Default|BindEventtoOnStartFiring (EventDispatchers|CreateEvent :!event Custom|OnStartFiring_Event))
  (Networking|CanExecuteCosmeticEvents))

(event EventOnSetTargeting
  (params (bNewIsTargeting :type Boolean))
  ;[1 1 1] Transition ADS/Hipfire SFX
  (bind _output_get (Variables|Default|SetPlayerIsADS bNewIsTargeting))
  (if _output_get
    (Audio|SpawnSoundAttached (Variables|Default|GetADSTransitionAudio) (Variables|Visual|GetWeaponMesh))
    (else
      (Audio|SpawnSoundAttached (Variables|Default|GetHipFireTransitionAudio) (Variables|Visual|GetWeaponMesh)))))

(event Custom|OnStartFiring_Event
  (if (Variables|Default|GetPlayerIsADS)
    (Niagara|SpawnSystemAttached "/FlexLegendWeaponGameplay/Assets/FX/NS_Weapon_Reload_Smoke.NS_Weapon_Reload_Smoke" (Variables|Visual|GetWeaponMesh) "Muzzle" "0, 0, 0" "0, 0, 0" "SnapToTargetIncludingScale" true)))

(fn OnRep_ADSMuzzleFlash)

(fn OnRep_HipFireMuzzleFlash)

(fn ConstructionScript
  (|CallParentFunction))
