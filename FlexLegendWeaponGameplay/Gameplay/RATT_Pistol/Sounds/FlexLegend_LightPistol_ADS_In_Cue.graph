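# Sound cue for the FlexLegend light pistol's aim-down-sights "in" foley in the RATT_Pistol weapon.
# It randomly picks one of three equally weighted wave variations, then passes it through a modulator with a raised pitch ceiling into a mixer whose second input is empty.
# The cue plays at reduced volume in the weapon foley sound class, reusing the weapon reload attenuation for spatial falloff.
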SoundCue /FlexLegendWeaponGameplay/Gameplay/RATT_Pistol/Sounds/FlexLegend_LightPistol_ADS_In_Cue
root=Mixer_0
nodes [
 {id=Mixer_0, children [Modulator_0,null], properties {inputVolume [1,0.25]}}
 {id=Modulator_0, children [Random_1], properties {pitchMax=1.2}}
 {id=Random_1, children [WavePlayer_0,WavePlayer_1,WavePlayer_2], properties {weights [1,1,1]}}
 {id=WavePlayer_0, properties {soundWaveAssetPtr=/FlexLegendWeaponGameplay/Gameplay/RATT_Pistol/Sounds/wavs/FlexLegend_LightPistol_ADS_In_01}}
 {id=WavePlayer_1, properties {soundWaveAssetPtr=/FlexLegendWeaponGameplay/Gameplay/RATT_Pistol/Sounds/wavs/FlexLegend_LightPistol_ADS_In_02}}
 {id=WavePlayer_2, properties {soundWaveAssetPtr=/FlexLegendWeaponGameplay/Gameplay/RATT_Pistol/Sounds/wavs/FlexLegend_LightPistol_ADS_In_03}}
]
properties {
 attenuationSettings=&Fort_Weapon_Reload_Att
 maxDistance=2250
 soundClassObject=&Fort_WeaponFoley
 volumeMultiplier=0.2
}
