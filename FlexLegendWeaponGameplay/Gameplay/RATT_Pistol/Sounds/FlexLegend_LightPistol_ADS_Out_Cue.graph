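# Sound cue for the RATT light pistol's aim-down-sights exit foley.
# A random node picks one of three equally weighted ADS-out wave variations, then a modulator applies a slight random upward pitch shift.
# It plays quietly in the weapon foley sound class and borrows the weapon reload attenuation settings for spatial falloff.
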SoundCue /FlexLegendWeaponGameplay/Gameplay/RATT_Pistol/Sounds/FlexLegend_LightPistol_ADS_Out_Cue
root=Modulator_0
nodes [
 {id=Modulator_0, children [Random_1], properties {pitchMax=1.2, pitchMin=1}}
 {id=Random_1, children [WavePlayer_0,WavePlayer_1,WavePlayer_2], properties {weights [1,1,1]}}
 {id=WavePlayer_0, properties {soundWaveAssetPtr=/FlexLegendWeaponGameplay/Gameplay/RATT_Pistol/Sounds/wavs/FlexLegend_LightPistol_ADS_Out_01}}
 {id=WavePlayer_1, properties {soundWaveAssetPtr=/FlexLegendWeaponGameplay/Gameplay/RATT_Pistol/Sounds/wavs/FlexLegend_LightPistol_ADS_Out_02}}
 {id=WavePlayer_2, properties {soundWaveAssetPtr=/FlexLegendWeaponGameplay/Gameplay/RATT_Pistol/Sounds/wavs/FlexLegend_LightPistol_ADS_Out_03}}
]
properties {
 attenuationSettings=&Fort_Weapon_Reload_Att
 maxDistance=2250
 soundClassObject=&Fort_WeaponFoley
 volumeMultiplier=0.2
}
