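# Sound cue for the equip sound of the FlexLegend light pistol (RATT_Pistol), classed as weapon foley at reduced volume.
# One of three equally weighted wave variations is picked at random and passed through a shared modulator.
# A branch on IsLocal sends true through an attenuation node with no assigned asset, while false or unset uses the weapon reload attenuation.
# The debug flag is left on and concurrency is overridden.
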SoundCue /FlexLegendWeaponGameplay/Gameplay/RATT_Pistol/Sounds/FlexLegend_LightPistol_Equip_Cue
root=Branch_0
nodes [
 {id=Attenuation_0, children [Modulator_0]}
 {id=Attenuation_1, children [Modulator_0], properties {attenuationSettings=&Fort_Weapon_Reload_Att}}
 {id=Branch_0, children [Attenuation_0, Attenuation_1, Attenuation_1], properties {boolParameterName=IsLocal}}
 {id=Modulator_0, children [Random_0], properties {pitchMax=1, volumeMax=1}}
 {id=Random_0, children [WavePlayer_0,WavePlayer_1,WavePlayer_2], properties {weights [1,1,1]}}
 {id=WavePlayer_0, properties {soundWaveAssetPtr=/FlexLegendWeaponGameplay/Gameplay/RATT_Pistol/Sounds/wavs/FlexLegend_LightPistol_Equip_01}}
 {id=WavePlayer_1, properties {soundWaveAssetPtr=/FlexLegendWeaponGameplay/Gameplay/RATT_Pistol/Sounds/wavs/FlexLegend_LightPistol_Equip_02}}
 {id=WavePlayer_2, properties {soundWaveAssetPtr=/FlexLegendWeaponGameplay/Gameplay/RATT_Pistol/Sounds/wavs/FlexLegend_LightPistol_Equip_03}}
]
properties {
 bDebug=true
 bOverrideConcurrency=true
 maxDistance=2097152
 soundClassObject=&Fort_WeaponFoley
 volumeMultiplier=0.4
}
