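# Sound cue for picking up the FlexLegend RATT light pistol.
# A Random node picks one of three equally weighted pickup waves, then a Modulator applies very slight volume variation.
# The cue plays at reduced volume in the loot-collect sound class with shared weapon-pickup attenuation, and the PickupMaximum_Oldest concurrency set limits overlapping pickups.
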SoundCue /FlexLegendWeaponGameplay/Gameplay/RATT_Pistol/Sounds/FlexLegend_LightPistol_Pickup_Cue
root=Modulator_0
nodes [
 {id=Modulator_0, children [Random_0], properties {volumeMax=1.01, volumeMin=0.99}}
 {id=Random_0, children [WavePlayer_0,WavePlayer_1,WavePlayer_2], properties {weights [1,1,1]}}
 {id=WavePlayer_0, properties {soundWaveAssetPtr=/FlexLegendWeaponGameplay/Gameplay/RATT_Pistol/Sounds/wavs/FlexLegend_LightPistol_PickUp_01}}
 {id=WavePlayer_1, properties {soundWaveAssetPtr=/FlexLegendWeaponGameplay/Gameplay/RATT_Pistol/Sounds/wavs/FlexLegend_LightPistol_PickUp_02}}
 {id=WavePlayer_2, properties {soundWaveAssetPtr=/FlexLegendWeaponGameplay/Gameplay/RATT_Pistol/Sounds/wavs/FlexLegend_LightPistol_PickUp_03}}
]
properties {
 attenuationSettings=&Fort_Weapon_Pickup_Att
 maxDistance=3500
 soundClassObject=&Fort_Loot_Collect
 volumeMultiplier=0.7
 concurrencyOverrides {maxCount=1, resolutionRule=PreventNew}
 concurrencySet [&PickupMaximum_Oldest]
}
