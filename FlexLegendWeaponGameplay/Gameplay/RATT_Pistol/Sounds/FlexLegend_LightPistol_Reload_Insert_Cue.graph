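# Sound cue for the magazine insert foley during the RATT light pistol reload.
# A random node picks one of three equally weighted insert wave variations, then a modulator adds slight pitch and volume variation.
# It plays through the weapon foley sound class with weapon reload attenuation, a reduced volume, and overridden concurrency, and has debug enabled.
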SoundCue /FlexLegendWeaponGameplay/Gameplay/RATT_Pistol/Sounds/FlexLegend_LightPistol_Reload_Insert_Cue
root=Modulator_0
nodes [
 {id=Modulator_0, children [Random_0], properties {pitchMax=1.02, pitchMin=0.99, volumeMax=1.02, volumeMin=0.98}}
 {id=Random_0, children [WavePlayer_0,WavePlayer_1,WavePlayer_2], properties {weights [1,1,1]}}
 {
  id=WavePlayer_0
  properties {soundWaveAssetPtr=/FlexLegendWeaponGameplay/Gameplay/RATT_Pistol/Sounds/wavs/FlexLegend_LightPistol_Reload_Insert_01}
 }
 {
  id=WavePlayer_1
  properties {soundWaveAssetPtr=/FlexLegendWeaponGameplay/Gameplay/RATT_Pistol/Sounds/wavs/FlexLegend_LightPistol_Reload_Insert_02}
 }
 {
  id=WavePlayer_2
  properties {soundWaveAssetPtr=/FlexLegendWeaponGameplay/Gameplay/RATT_Pistol/Sounds/wavs/FlexLegend_LightPistol_Reload_Insert_03}
 }
]
properties {
 attenuationSettings=&Fort_Weapon_Reload_Att
 bDebug=true
 bOverrideConcurrency=true
 maxDistance=2250
 soundClassObject=&Fort_WeaponFoley
 volumeMultiplier=0.6
}
